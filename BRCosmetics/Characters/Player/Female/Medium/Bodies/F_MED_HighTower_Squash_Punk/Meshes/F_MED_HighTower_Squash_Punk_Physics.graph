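# Physics asset for the HighTower Squash Punk female medium body cosmetic, covering a five-piece collar around the head and upper spine.
# Each collar bone has a thin box body constrained to spine_05 with tight angular limits, looser on the outer left and right pieces than on the middle and center ones.
# Capsules on the head and spine_05 approximate the body, with the head-to-spine constraint left free in swing and twist.
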PhysicsAsset /BRCosmetics/Characters/Player/Female/Medium/Bodies/F_MED_HighTower_Squash_Punk/Meshes/F_MED_HighTower_Squash_Punk_Physics
bodies {
 collar_c [{shapeType=Box, extentX=8.5, extentY=1.5, extentZ=10.5, center=<-2.4357 -0.8815 0>, rotation=<rot 0 22.5369 0>}]
 collar_l [{shapeType=Box, extentX=8.5, extentY=1.5, extentZ=7.5, center=<-1.244 0 0>}]
 collar_mid_l [{shapeType=Box, extentX=8.5, extentY=1.5, extentZ=11.5, center=<-1.5922 0 0.0001>, rotation=<rot -3.5074 0 0>}]
 collar_mid_r [{shapeType=Box, extentX=8.5, extentY=1.5, extentZ=11.5, center=<-1.5922 0 -0.0001>, rotation=<rot -3.5074 0 -180>}]
 collar_r [{shapeType=Box, extentX=8.5, extentY=1.5, extentZ=8, center=<-1.244 0 0>, rotation=<rot 0 0 180>}]
 head [
  {shapeType=Capsule, length=5.75, radius=6, center=<1.6392 3.0273 0>, rotation=<rot 90 0 0>}
  {shapeType=Capsule, length=5.25, radius=4.25, center=<0.364 -5.5948 0.0173>, rotation=<rot -0.7871 -109.7125 -89.7168>}
 ]
 spine_05 [{shapeType=Capsule, length=9.7549, radius=9.7046, center=<1.0237 0.382 -0.1671>, rotation=<rot -1.0929 -90 -0.4658>}]
}
constraints [
 {
  bone1Name=head
  bone2Name=spine_05
  swing1LimitDegrees=45
  swing1Motion=Free
  swing2LimitDegrees=45
  swing2Motion=Free
  twistLimitDegrees=45
  twistMotion=Free
 }
 {bone1Name=collar_l, bone2Name=spine_05, swing1LimitDegrees=20, swing2LimitDegrees=10, twistLimitDegrees=15}
 {bone1Name=collar_mid_l, bone2Name=spine_05, swing1LimitDegrees=15, swing2LimitDegrees=5, twistLimitDegrees=5}
 {bone1Name=collar_c, bone2Name=spine_05, swing1LimitDegrees=15, swing2LimitDegrees=5, twistLimitDegrees=5}
 {bone1Name=collar_r, bone2Name=spine_05, swing1LimitDegrees=20, swing2LimitDegrees=10, twistLimitDegrees=15}
 {bone1Name=collar_mid_r, bone2Name=spine_05, swing1LimitDegrees=15, swing2LimitDegrees=5, twistLimitDegrees=5}
]
