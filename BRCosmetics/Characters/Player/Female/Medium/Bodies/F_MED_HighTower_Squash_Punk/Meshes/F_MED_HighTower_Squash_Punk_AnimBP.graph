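# Animation Blueprint for the female HighTower Squash Punk cosmetic body, adding secondary motion and corrective deformation to the pose copied from its attached parent mesh.
# AnimDynamics drives lapels, jacket, zippers and necklace, with constraints placing zippers and pendant, followed by a RigidBody pass.
# Rotation values drive range-mapped thigh, collar and elbow corrective bone offsets.
# Collar dynamics, front-end elbow idle overlay and montage-tracking logic exist but appear unwired.
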AnimBlueprint /BRCosmetics/Characters/Player/Female/Medium/Bodies/F_MED_HighTower_Squash_Punk/Meshes/F_MED_HighTower_Squash_Punk_AnimBP
target_skeleton=&/F_MED_HighTower_Squash_Punk_Skeleton
graphs [
 AnimGraph {nodes [
   n1 = Anim Dynamics - Bone: collar_l(Alpha=1, ExternalForce="0,0,0") {
    angularDampingOverride=1
    angularSpringConstant=1500
    bAngularSpring=true
    bGravityOverrideInSimSpace=true
    bOverrideAngularDamping=true
    gravityScale=0.35
    simulationSpace=RootRelative
    alphaBoolBlend {blendOption=Linear}
    alphaScaleBiasClamp {scale=0.3}
    boundBone {boneName=collar_l}
    physicsBodyDefinitions [{
      sphereCollisionRadius=0
      boundBone {boneName=collar_l}
      boxExtents=<10 20 20>
      constraintSetup {angularLimitsMax=<0 0 0>, angularLimitsMin=<0 0 -40>, angularTarget=<0 0 0>}
      localJointOffset=<10 0 0>
     }]
   }
   n2 = Anim Dynamics - Bone: collar_c(ComponentPose=@n1, Alpha=1, ExternalForce="0,0,0") {
    angularDampingOverride=1
    angularSpringConstant=1500
    bAngularSpring=true
    bGravityOverrideInSimSpace=true
    bOverrideAngularDamping=true
    gravityScale=0.35
    simulationSpace=RootRelative
    alphaBoolBlend {blendOption=Linear}
    alphaScaleBiasClamp {scale=0.3}
    boundBone {boneName=collar_c}
    physicsBodyDefinitions [{
      sphereCollisionRadius=0
      boundBone {boneName=collar_c}
      boxExtents=<10 20 20>
      constraintSetup {angularLimitsMax=<0 0 0>, angularLimitsMin=<0 0 -40>, angularTarget=<0 0 0>}
      localJointOffset=<10 0 0>
     }]
   }
   Anim Dynamics - Bone: collar_r(ComponentPose=@n2, Alpha=1, ExternalForce="0,0,0") {
    angularDampingOverride=1
    angularSpringConstant=1500
    bAngularSpring=true
    bGravityOverrideInSimSpace=true
    bOverrideAngularDamping=true
    gravityScale=0.35
    simulationSpace=RootRelative
    alphaBoolBlend {blendOption=Linear}
    alphaScaleBiasClamp {scale=0.3}
    boundBone {boneName=collar_r}
    physicsBodyDefinitions [{
      sphereCollisionRadius=0
      boundBone {boneName=collar_r}
      boxExtents=<10 20 20>
      constraintSetup {angularLimitsMax=<0 0 0>, angularLimitsMin=<0 0 -40>, angularTarget=<0 0 0>}
      localJointOffset=<10 0 0>
     }]
   }
   Blend Poses by bool(BlendPose_0=Use cached pose 'frontEndPose'(), bActiveValue=Get bIsInFrontEnd()) {blendType=Linear}
   n3 = "Constraint - thigh_volume_l (R)"(ComponentPose=Local To Component(LocalPose=Copy Pose From Mesh() {bUseAttachedParent=true}), Alpha=0.5, ConstraintWeights_0=1) {
    alphaBoolBlend {blendOption=Linear}
    boneToModify {boneName=thigh_volume_l}
    constraintSetup [{transformType=Rotation, targetBone {boneName=thigh_l}}]
    constraintWeights [1]
   }
   n4 = "Constraint - thigh_volume_r (R)"(ComponentPose=@n3, Alpha=0.5, ConstraintWeights_0=1) {
    alphaBoolBlend {blendOption=Linear}
    boneToModify {boneName=thigh_volume_r}
    constraintSetup [{transformType=Rotation, targetBone {boneName=thigh_r}}]
    constraintWeights [1]
   }
   n5 = "Transform (Modify) Bone - Bone: thigh_volume_l"(ComponentPose=@n4, Alpha=Get ThighRotationLeft() -> (ThighRotationLeft:float), Translation="0,-3,-1", Rotation="0,0,0", Scale="1,1,1") {
    rotationMode=BMM_Additive
    rotationSpace=BCS_BoneSpace
    scaleMode=BMM_Additive
    scaleSpace=BCS_BoneSpace
    translationMode=BMM_Additive
    translationSpace=BCS_BoneSpace
    alphaBoolBlend {blendOption=Linear}
    alphaScaleBiasClamp {bMapRange=true, inRange {max=0, min=-100}, outRange {max=0, min=1}}
    boneToModify {boneName=thigh_volume_l}
    translation {y=-3, z=-1}
   }
   n6 = "Transform (Modify) Bone - Bone: thigh_volume_r"(ComponentPose=@n5, Alpha=Get ThighRotationRight() -> (ThighRotationRight:float), Translation="0,3,1", Rotation="0,0,0", Scale="1,1,1") {
    translationMode=BMM_Additive
    translationSpace=BCS_BoneSpace
    alphaBoolBlend {blendOption=Linear}
    alphaScaleBiasClamp {bMapRange=true, inRange {max=0, min=-100}, outRange {max=0, min=1}}
    boneToModify {boneName=thigh_volume_r}
    translation {y=3, z=1}
   }
   n7 = Anim Dynamics - Bone: lapel_r(ComponentPose=@n6, Alpha=1, ExternalForce="0,0,0") {
    angularDampingOverride=1
    angularSpringConstant=1500
    bAngularSpring=true
    bGravityOverrideInSimSpace=true
    bOverrideAngularDamping=true
    gravityScale=0.35
    simulationSpace=RootRelative
    alphaBoolBlend {blendOption=Linear}
    boundBone {boneName=lapel_r}
    physicsBodyDefinitions [{
      sphereCollisionRadius=0
      boundBone {boneName=lapel_r}
      boxExtents=<10 20 20>
      constraintSetup {angularLimitsMax=<0 0 0>, angularLimitsMin=<0 0 -40>, angularTarget=<0 0 0>}
      localJointOffset=<10 0 0>
     }]
   }
   n8 = Anim Dynamics - Bone: lapel_l(ComponentPose=@n7, Alpha=1, ExternalForce="0,0,0") {
    angularDampingOverride=1
    angularSpringConstant=1500
    bAngularSpring=true
    bGravityOverrideInSimSpace=true
    bOverrideAngularDamping=true
    gravityScale=0.35
    simulationSpace=RootRelative
    alphaBoolBlend {blendOption=Linear}
    boundBone {boneName=lapel_l}
    physicsBodyDefinitions [{
      sphereCollisionRadius=0
      boundBone {boneName=lapel_l}
      boxExtents=<10 20 20>
      constraintSetup {angularLimitsMax=<0 0 0>, angularLimitsMin=<0 0 -40>, angularTarget=<0 0 0>}
      localJointOffset=<10 0 0>
     }]
   }
   n9 = Anim Dynamics - Bone: jacket_1_c(ComponentPose=@n8, Alpha=1, ExternalForce="0,0,0") {
    angularDampingOverride=1
    angularSpringConstant=1500
    bAngularSpring=true
    bGravityOverrideInSimSpace=true
    bOverrideAngularDamping=true
    gravityScale=0.35
    simulationSpace=RootRelative
    alphaBoolBlend {blendOption=Linear}
    boundBone {boneName=jacket_1_c}
    physicsBodyDefinitions [{
      sphereCollisionRadius=0
      boundBone {boneName=jacket_1_c}
      boxExtents=<10 20 20>
      constraintSetup {angularLimitsMax=<0 0 0>, angularLimitsMin=<0 0 -40>, angularTarget=<0 0 0>}
      localJointOffset=<10 0 0>
     }]
   }
   n10 = Anim Dynamics - Bone: jacket_2_c(ComponentPose=@n9, Alpha=1, ExternalForce="0,0,0") {
    angularDampingOverride=1
    angularSpringConstant=1500
    bAngularSpring=true
    bGravityOverrideInSimSpace=true
    bOverrideAngularDamping=true
    gravityScale=0.35
    simulationSpace=RootRelative
    alphaBoolBlend {blendOption=Linear}
    alphaScaleBiasClamp {scale=0.3}
    boundBone {boneName=jacket_2_c}
    componentLinearAccScale {y=0.46, z=0.46}
    componentLinearVelScale {y=0.46, z=0.46}
    physicsBodyDefinitions [{
      sphereCollisionRadius=0
      boundBone {boneName=jacket_2_c}
      boxExtents=<10 20 20>
      constraintSetup {angularLimitsMax=<0 40 15>, angularLimitsMin=<0 -40 -90>, angularTarget=<0 0 0>}
      localJointOffset=<10 0 0>
     }]
   }
   n11 = Anim Dynamics - Bone: jacket_1_r(ComponentPose=@n10, Alpha=1, ExternalForce="0,0,0") {
    angularDampingOverride=1
    angularSpringConstant=1500
    bAngularSpring=true
    bGravityOverrideInSimSpace=true
    bOverrideAngularDamping=true
    gravityScale=0.35
    simulationSpace=RootRelative
    alphaBoolBlend {blendOption=Linear}
    alphaScaleBiasClamp {scale=0.3}
    boundBone {boneName=jacket_1_r}
    componentLinearAccScale {y=0.46, z=0.46}
    componentLinearVelScale {y=0.46, z=0.46}
    physicsBodyDefinitions [{
      sphereCollisionRadius=0
      boundBone {boneName=jacket_1_r}
      boxExtents=<10 20 20>
      constraintSetup {angularLimitsMax=<0 40 15>, angularLimitsMin=<0 -40 -90>, angularTarget=<0 0 0>}
      localJointOffset=<10 0 0>
     }]
   }
   n12 = Anim Dynamics - Bone: jacket_2_r(ComponentPose=@n11, Alpha=1, ExternalForce="0,0,0") {
    angularDampingOverride=1
    angularSpringConstant=1500
    bAngularSpring=true
    bGravityOverrideInSimSpace=true
    bOverrideAngularDamping=true
    gravityScale=0.35
    simulationSpace=RootRelative
    alphaBoolBlend {blendOption=Linear}
    alphaScaleBiasClamp {scale=0.3}
    boundBone {boneName=jacket_2_r}
    componentLinearAccScale {y=0.46, z=0.46}
    componentLinearVelScale {y=0.46, z=0.46}
    physicsBodyDefinitions [{
      sphereCollisionRadius=0
      boundBone {boneName=jacket_2_r}
      boxExtents=<10 20 20>
      constraintSetup {angularLimitsMax=<0 40 15>, angularLimitsMin=<0 -40 -90>, angularTarget=<0 0 0>}
      localJointOffset=<10 0 0>
     }]
   }
   n13 = Anim Dynamics - Bone: jacket_1_l(ComponentPose=@n12, Alpha=1, ExternalForce="0,0,0") {
    angularDampingOverride=1
    angularSpringConstant=1500
    bAngularSpring=true
    bGravityOverrideInSimSpace=true
    bOverrideAngularDamping=true
    gravityScale=0.35
    simulationSpace=RootRelative
    alphaBoolBlend {blendOption=Linear}
    alphaScaleBiasClamp {scale=0.3}
    boundBone {boneName=jacket_1_l}
    componentLinearAccScale {y=0.46, z=0.46}
    componentLinearVelScale {y=0.46, z=0.46}
    physicsBodyDefinitions [{
      sphereCollisionRadius=0
      boundBone {boneName=jacket_1_l}
      boxExtents=<10 20 20>
      constraintSetup {angularLimitsMax=<0 40 15>, angularLimitsMin=<0 -40 -90>, angularTarget=<0 0 0>}
      localJointOffset=<10 0 0>
     }]
   }
   n14 = Anim Dynamics - Bone: jacket_2_l(ComponentPose=@n13, Alpha=1, ExternalForce="0,0,0") {
    angularDampingOverride=1
    angularSpringConstant=1500
    bAngularSpring=true
    bGravityOverrideInSimSpace=true
    bOverrideAngularDamping=true
    gravityScale=0.35
    simulationSpace=RootRelative
    alphaBoolBlend {blendOption=Linear}
    alphaScaleBiasClamp {scale=0.3}
    boundBone {boneName=jacket_2_l}
    componentLinearAccScale {y=0.46, z=0.46}
    componentLinearVelScale {y=0.46, z=0.46}
    physicsBodyDefinitions [{
      sphereCollisionRadius=0
      boundBone {boneName=jacket_2_l}
      boxExtents=<10 20 20>
      constraintSetup {angularLimitsMax=<0 40 15>, angularLimitsMin=<0 -40 -90>, angularTarget=<0 0 0>}
      localJointOffset=<10 0 0>
     }]
   }
   n15 = "Constraint - zipper (PPPPP)"!(
    ComponentPose=@n14
    Alpha=1
    ConstraintWeights_0=0.1938
    ConstraintWeights_1:float=0.0396
    ConstraintWeights_2:float=0.0735
    ConstraintWeights_3:float=0.6003
    ConstraintWeights_4:float=0.0928
   ) -> (Pose) {
    alphaBoolBlend {blendOption=Linear}
    boneToModify {boneName=zipper}
    constraintSetup [
     {targetBone {boneName=spine_01}}
     {targetBone {boneName=spine_03}}
     {targetBone {boneName=spine_04}}
     {targetBone {boneName=jacket_2_l}}
     {targetBone {boneName=jacket_1_l}}
    ]
    constraintWeights [1,1,1,1,1]
   }
   n16 = Anim Dynamics - Bone: zipper(ComponentPose=@n15, Alpha=1, ExternalForce="0,0,0") {
    angularDampingOverride=1
    angularSpringConstant=1500
    bAngularSpring=true
    bGravityOverrideInSimSpace=true
    bOverrideAngularDamping=true
    gravityScale=0.35
    simulationSpace=RootRelative
    alphaBoolBlend {blendOption=Linear}
    alphaScaleBiasClamp {scale=0.3}
    boundBone {boneName=zipper}
    componentLinearAccScale {y=0.46, z=0.46}
    componentLinearVelScale {y=0.46, z=0.46}
    physicsBodyDefinitions [{
      sphereCollisionRadius=0
      boundBone {boneName=zipper}
      boxExtents=<10 20 20>
      constraintSetup {angularLimitsMax=<0 40 15>, angularLimitsMin=<0 -40 -90>, angularTarget=<0 0 0>}
      localJointOffset=<10 0 0>
     }]
   }
   n17 = "Constraint - calfZipper_l (PP)"!(ComponentPose=@n16, Alpha=1, ConstraintWeights_0=0.4325, ConstraintWeights_1:float=0.5675) -> (Pose) {
    alphaBoolBlend {blendOption=Linear}
    boneToModify {boneName=calfZipper_l}
    constraintSetup [{targetBone {boneName=calf_l}}, {targetBone {boneName=calf_twist_02_l}}]
    constraintWeights [1,1]
   }
   n18 = "Constraint - calfZipper_r (PP)"!(ComponentPose=@n17, Alpha=1, ConstraintWeights_0=0.5675, ConstraintWeights_1:float=0.4325) -> (Pose) {
    alphaBoolBlend {blendOption=Linear}
    boneToModify {boneName=calfZipper_r}
    constraintSetup [{targetBone {boneName=calf_twist_02_r}}, {targetBone {boneName=calf_r}}]
    constraintWeights [1,1]
   }
   n19 = Anim Dynamics - Bone: calfZipper_r(ComponentPose=@n18, Alpha=1, ExternalForce="0,0,0") {
    angularDampingOverride=1
    angularSpringConstant=1500
    bAngularSpring=true
    bGravityOverrideInSimSpace=true
    bOverrideAngularDamping=true
    gravityScale=0.35
    simulationSpace=RootRelative
    alphaBoolBlend {blendOption=Linear}
    alphaScaleBiasClamp {scale=0.4}
    boundBone {boneName=calfZipper_r}
    componentLinearAccScale {y=0.46, z=0.46}
    componentLinearVelScale {y=0.46, z=0.46}
    physicsBodyDefinitions [{
      sphereCollisionRadius=0
      boundBone {boneName=calfZipper_r}
      boxExtents=<10 20 20>
      constraintSetup {angularLimitsMax=<0 30 0>, angularLimitsMin=<0 -30 -90>, angularTarget=<0 0 0>}
      localJointOffset=<10 0 0>
     }]
   }
   n20 = Anim Dynamics - Bone: calfZipper_l(ComponentPose=@n19, Alpha=1, ExternalForce="0,0,0") {
    angularDampingOverride=1
    angularSpringConstant=1500
    bAngularSpring=true
    bGravityOverrideInSimSpace=true
    bOverrideAngularDamping=true
    gravityScale=0.35
    simulationSpace=RootRelative
    alphaBoolBlend {blendOption=Linear}
    alphaScaleBiasClamp {scale=0.4}
    boundBone {boneName=calfZipper_l}
    componentLinearAccScale {y=0.46, z=0.46}
    componentLinearVelScale {y=0.46, z=0.46}
    physicsBodyDefinitions [{
      sphereCollisionRadius=0
      boundBone {boneName=calfZipper_l}
      boxExtents=<10 20 20>
      constraintSetup {angularLimitsMax=<0 30 0>, angularLimitsMin=<0 -30 -90>, angularTarget=<0 0 0>}
      localJointOffset=<10 0 0>
     }]
   }
   n21 = Anim Dynamics - Bone: necklace_1(ComponentPose=@n20, Alpha=1, ExternalForce="0,0,0") {
    angularDampingOverride=1
    angularSpringConstant=1500
    bAngularSpring=true
    bGravityOverrideInSimSpace=true
    bOverrideAngularDamping=true
    gravityScale=0.35
    simulationSpace=RootRelative
    alphaBoolBlend {blendOption=Linear}
    alphaScaleBiasClamp {scale=0.3}
    boundBone {boneName=necklace_1}
    componentLinearAccScale {y=0.46, z=0.46}
    componentLinearVelScale {y=0.46, z=0.46}
    physicsBodyDefinitions [{
      sphereCollisionRadius=0
      boundBone {boneName=necklace_1}
      boxExtents=<10 20 20>
      constraintSetup {angularLimitsMax=<0 15 -5>, angularLimitsMin=<0 -15 -90>, angularTarget=<0 0 0>}
      localJointOffset=<10 0 0>
     }]
   }
   n22 = Anim Dynamics - Bone: necklace_2(ComponentPose=@n21, Alpha=1, ExternalForce="0,0,0") {
    angularDampingOverride=1
    angularSpringConstant=1500
    bAngularSpring=true
    bGravityOverrideInSimSpace=true
    bOverrideAngularDamping=true
    gravityScale=0.35
    simulationSpace=RootRelative
    alphaBoolBlend {blendOption=Linear}
    alphaScaleBiasClamp {scale=0.3}
    boundBone {boneName=necklace_2}
    componentLinearAccScale {y=0.46, z=0.46}
    componentLinearVelScale {y=0.46, z=0.46}
    physicsBodyDefinitions [{
      sphereCollisionRadius=0
      boundBone {boneName=necklace_2}
      boxExtents=<10 20 20>
      constraintSetup {angularLimitsMax=<0 30 0>, angularLimitsMin=<0 -30 -90>, angularTarget=<0 0 0>}
      localJointOffset=<10 0 0>
     }]
   }
   n23 = Anim Dynamics - Bone: necklace_3(ComponentPose=@n22, Alpha=1, ExternalForce="0,0,0") {
    angularDampingOverride=1
    angularSpringConstant=1500
    bAngularSpring=true
    bGravityOverrideInSimSpace=true
    bOverrideAngularDamping=true
    gravityScale=0.35
    simulationSpace=RootRelative
    alphaBoolBlend {blendOption=Linear}
    alphaScaleBiasClamp {scale=0.3}
    boundBone {boneName=necklace_3}
    componentLinearAccScale {y=0.46, z=0.46}
    componentLinearVelScale {y=0.46, z=0.46}
    physicsBodyDefinitions [{
      sphereCollisionRadius=0
      boundBone {boneName=necklace_3}
      boxExtents=<10 20 20>
      constraintSetup {angularLimitsMax=<0 30 0>, angularLimitsMin=<0 -30 -90>, angularTarget=<0 0 0>}
      localJointOffset=<10 0 0>
     }]
   }
   n24 = "Constraint - pendant (PPPP)"!(
    ComponentPose=@n23
    Alpha=1
    ConstraintWeights_0=0.2826
    ConstraintWeights_1:float=0.1676
    ConstraintWeights_2:float=0.1357
    ConstraintWeights_3:float=0.4141
   ) -> (Pose) {
    alphaBoolBlend {blendOption=Linear}
    boneToModify {boneName=pendant}
    constraintSetup [{targetBone {boneName=neck_02}}, {targetBone {boneName=neck_01}}, {targetBone {boneName=head}}, {targetBone {boneName=spine_05}}]
    constraintWeights [1,1,1,1]
   }
   n25 = Anim Dynamics - Bone: pendant(ComponentPose=@n24, Alpha=1, ExternalForce="0,0,0") {
    angularDampingOverride=1
    angularSpringConstant=1500
    bAngularSpring=true
    bGravityOverrideInSimSpace=true
    bOverrideAngularDamping=true
    gravityScale=0.35
    simulationSpace=RootRelative
    alphaBoolBlend {blendOption=Linear}
    alphaScaleBiasClamp {scale=0.3}
    boundBone {boneName=pendant}
    componentLinearAccScale {y=0.46, z=0.46}
    componentLinearVelScale {y=0.46, z=0.46}
    physicsBodyDefinitions [{
      sphereCollisionRadius=0
      boundBone {boneName=pendant}
      boxExtents=<10 20 20>
      constraintSetup {angularLimitsMax=<0 40 15>, angularLimitsMin=<0 -40 -90>, angularTarget=<0 0 0>}
      localJointOffset=<10 0 0>
     }]
   }
   n26 = RigidBody(ComponentPose=@n25) {
    bClampLinearTranslationLimitToRefPose=true
    bOverrideWorldGravity=true
    overridePhysicsAsset=&/F_MED_HighTower_Squash_Punk_Physics
    simulationSpace=BaseBoneSpace
    alphaBoolBlend {blendOption=Linear}
    baseBoneRef {boneName=root}
    componentLinearAccScale=<0.3 0.3 0.3>
    componentLinearVelScale=<0.3 0.3 0.3>
    overrideWorldGravity {z=-980}
   }
   n27 = Get HeadRotation() -> (HeadRotation:float)
   n28 = "Transform (Modify) Bone - Bone: collar_c"(ComponentPose=@n26, Alpha=@n27, Translation="0,0,0", Rotation="0,5,0", Scale="1,1,1") {
    rotationMode=BMM_Additive
    rotationSpace=BCS_BoneSpace
    scaleMode=BMM_Additive
    scaleSpace=BCS_BoneSpace
    translationMode=BMM_Additive
    translationSpace=BCS_BoneSpace
    alphaBoolBlend {blendOption=Linear}
    alphaScaleBiasClamp {bMapRange=true, inRange {max=5, min=-25}, outRange {max=0, min=1}}
    boneToModify {boneName=collar_c}
    rotation {yaw=5}
   }
   n29 = "Transform (Modify) Bone - Bone: collar_mid_l"(ComponentPose=@n28, Alpha=@n27, Translation="0,-1,0", Rotation="-5,15,0", Scale="1,1,1") {
    rotationMode=BMM_Additive
    rotationSpace=BCS_BoneSpace
    scaleMode=BMM_Additive
    scaleSpace=BCS_BoneSpace
    translationMode=BMM_Additive
    translationSpace=BCS_BoneSpace
    alphaBoolBlend {blendOption=Linear}
    alphaScaleBiasClamp {bMapRange=true, inRange {max=5, min=-25}, outRange {max=0, min=1}}
    boneToModify {boneName=collar_mid_l}
    rotation {pitch=-5, yaw=15}
    translation {y=-1}
   }
   n30 = "Transform (Modify) Bone - Bone: collar_mid_r"(ComponentPose=@n29, Alpha=@n27, Translation="0,-1.5,0", Rotation="-5,10,0", Scale="1,1,1") {
    rotationMode=BMM_Additive
    rotationSpace=BCS_BoneSpace
    scaleMode=BMM_Additive
    scaleSpace=BCS_BoneSpace
    translationMode=BMM_Additive
    translationSpace=BCS_BoneSpace
    alphaBoolBlend {blendOption=Linear}
    alphaScaleBiasClamp {bMapRange=true, inRange {max=5, min=-25}, outRange {max=0, min=1}}
    boneToModify {boneName=collar_mid_r}
    rotation {pitch=-5, yaw=10}
    translation {y=-1.5}
   }
   n31 = "Constraint - dyn_elbow_base_l (PP)"!(ComponentPose=@n30, Alpha=1, ConstraintWeights_0=0.6, ConstraintWeights_1:float=0.4) -> (Pose) {
    alphaBoolBlend {blendOption=Linear}
    boneToModify {boneName=dyn_elbow_base_l}
    constraintSetup [{targetBone {boneName=upperarm_l}}, {targetBone {boneName=lowerarm_l}}]
    constraintWeights [1,1]
   }
   n32 = "Transform (Modify) Bone - Bone: dyn_elbow_l"(ComponentPose=@n31, Alpha=Get ElbowRotationLeft() -> (ElbowRotationLeft:float), Translation="0,-3,0", Rotation="0,0,0", Scale="0.6,1,1") {
    rotationMode=BMM_Additive
    rotationSpace=BCS_BoneSpace
    scaleMode=BMM_Additive
    scaleSpace=BCS_BoneSpace
    translationMode=BMM_Additive
    translationSpace=BCS_BoneSpace
    alphaBoolBlend {blendOption=Linear}
    alphaScaleBiasClamp {bMapRange=true, inRange {max=100}}
    boneToModify {boneName=dyn_elbow_l}
    scale {x=0.6}
    translation {y=-3}
   }
   n33 = "Constraint - dyn_elbow_base_r (PP)"!(ComponentPose=@n32, Alpha=1, ConstraintWeights_0=0.6, ConstraintWeights_1:float=0.4) -> (Pose) {
    alphaBoolBlend {blendOption=Linear}
    boneToModify {boneName=dyn_elbow_base_r}
    constraintSetup [{targetBone {boneName=upperarm_r}}, {targetBone {boneName=lowerarm_r}}]
    constraintWeights [1,1]
   }
   n34 = "Transform (Modify) Bone - Bone: dyn_elbow_r"(ComponentPose=@n33, Alpha=Get ElbowRotationRight() -> (ElbowRotationRight:float), Translation="0,3,0", Rotation="0,0,0", Scale="0.7,1,1") {
    rotationMode=BMM_Additive
    rotationSpace=BCS_BoneSpace
    scaleMode=BMM_Additive
    scaleSpace=BCS_BoneSpace
    translationMode=BMM_Additive
    translationSpace=BCS_BoneSpace
    alphaBoolBlend {blendOption=Linear}
    alphaScaleBiasClamp {bMapRange=true, inRange {max=100}}
    boneToModify {boneName=dyn_elbow_r}
    scale {x=0.7}
    translation {y=3}
   }
   Output Pose(Result=Component To Local(ComponentPose=@n34)) {layerGroup=None}
   Save cached pose 'mainPose'() {cacheName=mainPose}
   n35 = Get Muscles_Position() -> (Muscles_Position:double)
   Save cached pose 'frontEndPose'(Pose=Layered blend per bone(
     BasePose=Copy Pose From Mesh() {bUseAttachedParent=true}
     BlendPoses_0="Sequence Evaluator (by time) 'HighTower_SquashPunk_Female_Idle'"!(ExplicitTime:float=@n35) -> (Pose:Pose Link Structure) {
      is=title AnimGraphNode_SequenceEvaluator
      sequence=/BRCosmetics/Animation/Game/MainPlayer/Menu/BR/CustomSkins/_S14/HighTower/HighTower_SquashPunk_Female_Idle
     }
     BlendWeights_0=Get Emote() -> (Emote:double)
    ) {layerSetup [{branchFilters [{boneName=dyn_elbow_l}, {boneName=dyn_elbow_r}]}]}) {cacheName=frontEndPose}
   Use cached pose 'mainPose'()
  ]}
 EventGraph {nodes [
   Reroute Node()
   n1 = Get Anim_Inst() -> (Anim_Inst:Anim Instance Object Reference)
   > Set Emote(Emote:double=SelectFloat(
     A=1
     B=0
     bPickA=Montage_IsActive!(
      self:Anim Instance Object Reference=@n1
      Montage:Anim Montage Object Reference=/BRCosmetics/Animation/Game/MainPlayer/Menu/BR/CustomSkins/_S14/HighTower/HighTower_Squash_Female_Idle_Montage
     ) -> (ReturnValue:bool)
    )) -> (Output_Get:double)
   > Set Muscles_Position(Muscles_Position:double=Montage_GetPosition!(
     self:Anim Instance Object Reference=@n1
     Montage:Anim Montage Object Reference=&HighTower_Squash_Female_Idle_Montage
    ) -> (ReturnValue:float)) -> (Output_Get:double)
   n2 = Set Skel_Mesh(Skel_Mesh=GetOwnerSkeletalMeshComponent())
   > Is Valid(InputObject=@n2)
   > Set Anim_Inst(Anim_Inst:Anim Instance Object Reference=GetAnimInstance(self=@n2)) -> (Output_Get:Anim Instance Object Reference)
   Event BlueprintUpdateAnimation()
   Event BlueprintInitializeAnimation()
  ]}
]
variables [
 Anim_Inst {type=AnimInstance, default=null}
 Emote {type=number}
 Muscles_Position {type=number}
 Skel_Mesh {type=SkeletalMeshComponent}
]
