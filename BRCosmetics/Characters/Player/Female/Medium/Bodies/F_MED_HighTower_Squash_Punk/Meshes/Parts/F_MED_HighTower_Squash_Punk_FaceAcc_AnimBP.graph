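# Post-copy animation blueprint for the FaceAcc part of the female HighTower Squash Punk cosmetic body, adding secondary motion to its hair bones.
# It copies the attached parent's pose, then additively scales hair6_1 and hair7_1 by HeadRotation, remapped inversely to alpha.
# Chained spring-based Anim Dynamics nodes then simulate seven hair groups at low blend weights, lighter at roots than further out.
# The EventGraph is empty.
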AnimBlueprint /BRCosmetics/Characters/Player/Female/Medium/Bodies/F_MED_HighTower_Squash_Punk/Meshes/Parts/F_MED_HighTower_Squash_Punk_FaceAcc_AnimBP
target_skeleton=&/F_MED_HighTower_Squash_Punk_FaceAcc_Skeleton
graphs [
 AnimGraph {nodes [
   n1 = Get HeadRotation() -> (HeadRotation:float)
   n2 = "Transform (Modify) Bone - Bone: hair6_1"(
    ComponentPose=Local To Component(LocalPose=Copy Pose From Mesh() {bUseAttachedParent=true})
    Alpha=@n1
    Translation="0,0,0"
    Rotation="0,0,0"
    Scale="0.45,0.45,0.45"
   ) {
    rotationMode=BMM_Additive
    rotationSpace=BCS_BoneSpace
    scaleMode=BMM_Additive
    scaleSpace=BCS_BoneSpace
    translationMode=BMM_Additive
    translationSpace=BCS_BoneSpace
    alphaBoolBlend {blendOption=Linear}
    alphaScaleBiasClamp {bMapRange=true, inRange {max=10, min=-25}, outRange {max=0, min=1}}
    boneToModify {boneName=hair6_1}
    scale=<0.45 0.45 0.45>
   }
   n3 = "Transform (Modify) Bone - Bone: hair7_1"(ComponentPose=@n2, Alpha=@n1, Translation="0,0,0", Rotation="0,0,0", Scale="0.7,0.7,0.7") {
    rotationMode=BMM_Additive
    rotationSpace=BCS_BoneSpace
    scaleMode=BMM_Additive
    scaleSpace=BCS_BoneSpace
    translationMode=BMM_Additive
    translationSpace=BCS_BoneSpace
    alphaBoolBlend {blendOption=Linear}
    alphaScaleBiasClamp {bMapRange=true, inRange {max=10, min=-25}, outRange {max=0, min=1}}
    boneToModify {boneName=hair7_1}
    scale=<0.7 0.7 0.7>
   }
   n4 = Anim Dynamics - Bone: hair1_1(ComponentPose=@n3, Alpha=1, ExternalForce="0,0,0") {
    angularDampingOverride=1
    angularSpringConstant=1500
    bAngularSpring=true
    bGravityOverrideInSimSpace=true
    bOverrideAngularDamping=true
    gravityScale=0.35
    simulationSpace=RootRelative
    alphaBoolBlend {blendOption=Linear}
    alphaScaleBiasClamp {scale=0.1}
    boundBone {boneName=hair1_1}
    componentLinearAccScale {y=0.46, z=0.46}
    componentLinearVelScale {y=0.46, z=0.46}
    physicsBodyDefinitions [{
      sphereCollisionRadius=0
      boundBone {boneName=hair1_1}
      boxExtents=<10 20 20>
      constraintSetup {angularLimitsMax=<0 90 90>, angularLimitsMin=<0 -90 -90>, angularTarget=<0 0 0>}
      localJointOffset=<10 0 0>
     }]
   }
   n5 = Anim Dynamics - Bone: hair1_2(ComponentPose=@n4, Alpha=1, ExternalForce="0,0,0") {
    angularDampingOverride=1
    angularSpringConstant=1500
    bAngularSpring=true
    bGravityOverrideInSimSpace=true
    bOverrideAngularDamping=true
    gravityScale=0.35
    simulationSpace=RootRelative
    alphaBoolBlend {blendOption=Linear}
    alphaScaleBiasClamp {scale=0.2}
    boundBone {boneName=hair1_2}
    componentLinearAccScale {y=0.46, z=0.46}
    componentLinearVelScale {y=0.46, z=0.46}
    physicsBodyDefinitions [{
      sphereCollisionRadius=0
      boundBone {boneName=hair1_2}
      boxExtents=<10 20 20>
      constraintSetup {angularLimitsMax=<0 90 90>, angularLimitsMin=<0 -90 -90>, angularTarget=<0 0 0>}
      localJointOffset=<10 0 0>
     }]
   }
   n6 = Anim Dynamics - Bone: hair1_3(ComponentPose=@n5, Alpha=1, ExternalForce="0,0,0") {
    angularDampingOverride=1
    angularSpringConstant=1500
    bAngularSpring=true
    bGravityOverrideInSimSpace=true
    bOverrideAngularDamping=true
    gravityScale=0.35
    simulationSpace=RootRelative
    alphaBoolBlend {blendOption=Linear}
    alphaScaleBiasClamp {scale=0.3}
    boundBone {boneName=hair1_3}
    componentLinearAccScale {y=0.46, z=0.46}
    componentLinearVelScale {y=0.46, z=0.46}
    physicsBodyDefinitions [{
      sphereCollisionRadius=0
      boundBone {boneName=hair1_3}
      boxExtents=<10 20 20>
      constraintSetup {angularLimitsMax=<0 90 90>, angularLimitsMin=<0 -90 -90>, angularTarget=<0 0 0>}
      localJointOffset=<10 0 0>
     }]
   }
   n7 = Anim Dynamics - Bone: hair2_1(ComponentPose=@n6, Alpha=1, ExternalForce="0,0,0") {
    angularDampingOverride=1
    angularSpringConstant=1500
    bAngularSpring=true
    bGravityOverrideInSimSpace=true
    bOverrideAngularDamping=true
    gravityScale=0.35
    simulationSpace=RootRelative
    alphaBoolBlend {blendOption=Linear}
    alphaScaleBiasClamp {scale=0.1}
    boundBone {boneName=hair2_1}
    componentLinearAccScale {y=0.46, z=0.46}
    componentLinearVelScale {y=0.46, z=0.46}
    physicsBodyDefinitions [{
      sphereCollisionRadius=0
      boundBone {boneName=hair2_1}
      boxExtents=<10 20 20>
      constraintSetup {angularLimitsMax=<0 90 90>, angularLimitsMin=<0 -90 -90>, angularTarget=<0 0 0>}
      localJointOffset=<10 0 0>
     }]
   }
   n8 = Anim Dynamics - Bone: hair2A_2(ComponentPose=@n7, Alpha=1, ExternalForce="0,0,0") {
    angularDampingOverride=1
    angularSpringConstant=1500
    bAngularSpring=true
    bGravityOverrideInSimSpace=true
    bOverrideAngularDamping=true
    gravityScale=0.35
    simulationSpace=RootRelative
    alphaBoolBlend {blendOption=Linear}
    alphaScaleBiasClamp {scale=0.2}
    boundBone {boneName=hair2A_2}
    componentLinearAccScale {y=0.46, z=0.46}
    componentLinearVelScale {y=0.46, z=0.46}
    physicsBodyDefinitions [{
      sphereCollisionRadius=0
      boundBone {boneName=hair2A_2}
      boxExtents=<10 20 20>
      constraintSetup {angularLimitsMax=<0 90 90>, angularLimitsMin=<0 -90 -90>, angularTarget=<0 0 0>}
      localJointOffset=<10 0 0>
     }]
   }
   n9 = Anim Dynamics - Bone: hair2B_2(ComponentPose=@n8, Alpha=1, ExternalForce="0,0,0") {
    angularDampingOverride=1
    angularSpringConstant=1500
    bAngularSpring=true
    bGravityOverrideInSimSpace=true
    bOverrideAngularDamping=true
    gravityScale=0.35
    simulationSpace=RootRelative
    alphaBoolBlend {blendOption=Linear}
    alphaScaleBiasClamp {scale=0.3}
    boundBone {boneName=hair2B_2}
    componentLinearAccScale {y=0.46, z=0.46}
    componentLinearVelScale {y=0.46, z=0.46}
    physicsBodyDefinitions [{
      sphereCollisionRadius=0
      boundBone {boneName=hair2B_2}
      boxExtents=<10 20 20>
      constraintSetup {angularLimitsMax=<0 90 90>, angularLimitsMin=<0 -90 -90>, angularTarget=<0 0 0>}
      localJointOffset=<10 0 0>
     }]
   }
   n10 = Anim Dynamics - Bone: hair3_1(ComponentPose=@n9, Alpha=1, ExternalForce="0,0,0") {
    angularDampingOverride=1
    angularSpringConstant=1500
    bAngularSpring=true
    bGravityOverrideInSimSpace=true
    bOverrideAngularDamping=true
    gravityScale=0.35
    simulationSpace=RootRelative
    alphaBoolBlend {blendOption=Linear}
    alphaScaleBiasClamp {scale=0.1}
    boundBone {boneName=hair3_1}
    componentLinearAccScale {y=0.46, z=0.46}
    componentLinearVelScale {y=0.46, z=0.46}
    physicsBodyDefinitions [{
      sphereCollisionRadius=0
      boundBone {boneName=hair3_1}
      boxExtents=<10 20 20>
      constraintSetup {angularLimitsMax=<0 90 90>, angularLimitsMin=<0 -90 -90>, angularTarget=<0 0 0>}
      localJointOffset=<10 0 0>
     }]
   }
   n11 = Anim Dynamics - Bone: hair3_2(ComponentPose=@n10, Alpha=1, ExternalForce="0,0,0") {
    angularDampingOverride=1
    angularSpringConstant=1500
    bAngularSpring=true
    bGravityOverrideInSimSpace=true
    bOverrideAngularDamping=true
    gravityScale=0.35
    simulationSpace=RootRelative
    alphaBoolBlend {blendOption=Linear}
    alphaScaleBiasClamp {scale=0.2}
    boundBone {boneName=hair3_2}
    componentLinearAccScale {y=0.46, z=0.46}
    componentLinearVelScale {y=0.46, z=0.46}
    physicsBodyDefinitions [{
      sphereCollisionRadius=0
      boundBone {boneName=hair3_2}
      boxExtents=<10 20 20>
      constraintSetup {angularLimitsMax=<0 90 90>, angularLimitsMin=<0 -90 -90>, angularTarget=<0 0 0>}
      localJointOffset=<10 0 0>
     }]
   }
   n12 = Anim Dynamics - Bone: hair3_3(ComponentPose=@n11, Alpha=1, ExternalForce="0,0,0") {
    angularDampingOverride=1
    angularSpringConstant=1500
    bAngularSpring=true
    bGravityOverrideInSimSpace=true
    bOverrideAngularDamping=true
    gravityScale=0.35
    simulationSpace=RootRelative
    alphaBoolBlend {blendOption=Linear}
    alphaScaleBiasClamp {scale=0.3}
    boundBone {boneName=hair3_3}
    componentLinearAccScale {y=0.46, z=0.46}
    componentLinearVelScale {y=0.46, z=0.46}
    physicsBodyDefinitions [{
      sphereCollisionRadius=0
      boundBone {boneName=hair3_3}
      boxExtents=<10 20 20>
      constraintSetup {angularLimitsMax=<0 90 90>, angularLimitsMin=<0 -90 -90>, angularTarget=<0 0 0>}
      localJointOffset=<10 0 0>
     }]
   }
   n13 = Anim Dynamics - Bone: hair4_1(ComponentPose=@n12, Alpha=1, ExternalForce="0,0,0") {
    angularDampingOverride=1
    angularSpringConstant=1500
    bAngularSpring=true
    bGravityOverrideInSimSpace=true
    bOverrideAngularDamping=true
    gravityScale=0.35
    simulationSpace=RootRelative
    alphaBoolBlend {blendOption=Linear}
    alphaScaleBiasClamp {scale=0.1}
    boundBone {boneName=hair4_1}
    componentLinearAccScale {y=0.46, z=0.46}
    componentLinearVelScale {y=0.46, z=0.46}
    physicsBodyDefinitions [{
      sphereCollisionRadius=0
      boundBone {boneName=hair4_1}
      boxExtents=<10 20 20>
      constraintSetup {angularLimitsMax=<0 90 90>, angularLimitsMin=<0 -90 -90>, angularTarget=<0 0 0>}
      localJointOffset=<10 0 0>
     }]
   }
   n14 = Anim Dynamics - Bone: hair4A_2(ComponentPose=@n13, Alpha=1, ExternalForce="0,0,0") {
    angularDampingOverride=1
    angularSpringConstant=1500
    bAngularSpring=true
    bGravityOverrideInSimSpace=true
    bOverrideAngularDamping=true
    gravityScale=0.35
    simulationSpace=RootRelative
    alphaBoolBlend {blendOption=Linear}
    alphaScaleBiasClamp {scale=0.2}
    boundBone {boneName=hair4A_2}
    componentLinearAccScale {y=0.46, z=0.46}
    componentLinearVelScale {y=0.46, z=0.46}
    physicsBodyDefinitions [{
      sphereCollisionRadius=0
      boundBone {boneName=hair4A_2}
      boxExtents=<10 20 20>
      constraintSetup {angularLimitsMax=<0 90 90>, angularLimitsMin=<0 -90 -90>, angularTarget=<0 0 0>}
      localJointOffset=<10 0 0>
     }]
   }
   n15 = Anim Dynamics - Bone: hair4B_2(ComponentPose=@n14, Alpha=1, ExternalForce="0,0,0") {
    angularDampingOverride=1
    angularSpringConstant=1500
    bAngularSpring=true
    bGravityOverrideInSimSpace=true
    bOverrideAngularDamping=true
    gravityScale=0.35
    simulationSpace=RootRelative
    alphaBoolBlend {blendOption=Linear}
    alphaScaleBiasClamp {scale=0.2}
    boundBone {boneName=hair4B_2}
    componentLinearAccScale {y=0.46, z=0.46}
    componentLinearVelScale {y=0.46, z=0.46}
    physicsBodyDefinitions [{
      sphereCollisionRadius=0
      boundBone {boneName=hair4B_2}
      boxExtents=<10 20 20>
      constraintSetup {angularLimitsMax=<0 90 90>, angularLimitsMin=<0 -90 -90>, angularTarget=<0 0 0>}
      localJointOffset=<10 0 0>
     }]
   }
   n16 = Anim Dynamics - Bone: hair4C_2(ComponentPose=@n15, Alpha=1, ExternalForce="0,0,0") {
    angularDampingOverride=1
    angularSpringConstant=1500
    bAngularSpring=true
    bGravityOverrideInSimSpace=true
    bOverrideAngularDamping=true
    gravityScale=0.35
    simulationSpace=RootRelative
    alphaBoolBlend {blendOption=Linear}
    alphaScaleBiasClamp {scale=0.2}
    boundBone {boneName=hair4C_2}
    componentLinearAccScale {y=0.46, z=0.46}
    componentLinearVelScale {y=0.46, z=0.46}
    physicsBodyDefinitions [{
      sphereCollisionRadius=0
      boundBone {boneName=hair4C_2}
      boxExtents=<10 20 20>
      constraintSetup {angularLimitsMax=<0 90 90>, angularLimitsMin=<0 -90 -90>, angularTarget=<0 0 0>}
      localJointOffset=<10 0 0>
     }]
   }
   n17 = Anim Dynamics - Bone: hair5_1(ComponentPose=@n16, Alpha=1, ExternalForce="0,0,0") {
    angularDampingOverride=1
    angularSpringConstant=1500
    bAngularSpring=true
    bGravityOverrideInSimSpace=true
    bOverrideAngularDamping=true
    gravityScale=0.35
    simulationSpace=RootRelative
    alphaBoolBlend {blendOption=Linear}
    alphaScaleBiasClamp {scale=0.1}
    boundBone {boneName=hair5_1}
    componentLinearAccScale {y=0.46, z=0.46}
    componentLinearVelScale {y=0.46, z=0.46}
    physicsBodyDefinitions [{
      sphereCollisionRadius=0
      boundBone {boneName=hair5_1}
      boxExtents=<10 20 20>
      constraintSetup {angularLimitsMax=<0 90 90>, angularLimitsMin=<0 -90 -90>, angularTarget=<0 0 0>}
      localJointOffset=<10 0 0>
     }]
   }
   n18 = Anim Dynamics - Bone: hair5A_2(ComponentPose=@n17, Alpha=1, ExternalForce="0,0,0") {
    angularDampingOverride=1
    angularSpringConstant=1500
    bAngularSpring=true
    bGravityOverrideInSimSpace=true
    bOverrideAngularDamping=true
    gravityScale=0.35
    simulationSpace=RootRelative
    alphaBoolBlend {blendOption=Linear}
    alphaScaleBiasClamp {scale=0.2}
    boundBone {boneName=hair5A_2}
    componentLinearAccScale {y=0.46, z=0.46}
    componentLinearVelScale {y=0.46, z=0.46}
    physicsBodyDefinitions [{
      sphereCollisionRadius=0
      boundBone {boneName=hair5A_2}
      boxExtents=<10 20 20>
      constraintSetup {angularLimitsMax=<0 90 90>, angularLimitsMin=<0 -90 -90>, angularTarget=<0 0 0>}
      localJointOffset=<10 0 0>
     }]
   }
   n19 = Anim Dynamics - Bone: hair5B_2(ComponentPose=@n18, Alpha=1, ExternalForce="0,0,0") {
    angularDampingOverride=1
    angularSpringConstant=1500
    bAngularSpring=true
    bGravityOverrideInSimSpace=true
    bOverrideAngularDamping=true
    gravityScale=0.35
    simulationSpace=RootRelative
    alphaBoolBlend {blendOption=Linear}
    alphaScaleBiasClamp {scale=0.3}
    boundBone {boneName=hair5B_2}
    componentLinearAccScale {y=0.46, z=0.46}
    componentLinearVelScale {y=0.46, z=0.46}
    physicsBodyDefinitions [{
      sphereCollisionRadius=0
      boundBone {boneName=hair5B_2}
      boxExtents=<10 20 20>
      constraintSetup {angularLimitsMax=<0 90 90>, angularLimitsMin=<0 -90 -90>, angularTarget=<0 0 0>}
      localJointOffset=<10 0 0>
     }]
   }
   n20 = Anim Dynamics - Bone: hair6_1(ComponentPose=@n19, Alpha=1, ExternalForce="0,0,0") {
    angularDampingOverride=1
    angularSpringConstant=1500
    bAngularSpring=true
    bGravityOverrideInSimSpace=true
    bOverrideAngularDamping=true
    gravityScale=0.35
    simulationSpace=RootRelative
    alphaBoolBlend {blendOption=Linear}
    alphaScaleBiasClamp {scale=0.1}
    boundBone {boneName=hair6_1}
    componentLinearAccScale {y=0.46, z=0.46}
    componentLinearVelScale {y=0.46, z=0.46}
    physicsBodyDefinitions [{
      sphereCollisionRadius=0
      boundBone {boneName=hair6_1}
      boxExtents=<10 20 20>
      constraintSetup {angularLimitsMax=<0 90 90>, angularLimitsMin=<0 -90 -90>, angularTarget=<0 0 0>}
      localJointOffset=<10 0 0>
     }]
   }
   n21 = Anim Dynamics - Bone: hair6A_2(ComponentPose=@n20, Alpha=1, ExternalForce="0,0,0") {
    angularDampingOverride=1
    angularSpringConstant=1500
    bAngularSpring=true
    bGravityOverrideInSimSpace=true
    bOverrideAngularDamping=true
    gravityScale=0.35
    simulationSpace=RootRelative
    alphaBoolBlend {blendOption=Linear}
    alphaScaleBiasClamp {scale=0.2}
    boundBone {boneName=hair6A_2}
    componentLinearAccScale {y=0.46, z=0.46}
    componentLinearVelScale {y=0.46, z=0.46}
    physicsBodyDefinitions [{
      sphereCollisionRadius=0
      boundBone {boneName=hair6A_2}
      boxExtents=<10 20 20>
      constraintSetup {angularLimitsMax=<0 90 90>, angularLimitsMin=<0 -90 -90>, angularTarget=<0 0 0>}
      localJointOffset=<10 0 0>
     }]
   }
   n22 = Anim Dynamics - Bone: hair6B_2(ComponentPose=@n21, Alpha=1, ExternalForce="0,0,0") {
    angularDampingOverride=1
    angularSpringConstant=1500
    bAngularSpring=true
    bGravityOverrideInSimSpace=true
    bOverrideAngularDamping=true
    gravityScale=0.35
    simulationSpace=RootRelative
    alphaBoolBlend {blendOption=Linear}
    alphaScaleBiasClamp {scale=0.3}
    boundBone {boneName=hair6B_2}
    componentLinearAccScale {y=0.46, z=0.46}
    componentLinearVelScale {y=0.46, z=0.46}
    physicsBodyDefinitions [{
      sphereCollisionRadius=0
      boundBone {boneName=hair6B_2}
      boxExtents=<10 20 20>
      constraintSetup {angularLimitsMax=<0 90 90>, angularLimitsMin=<0 -90 -90>, angularTarget=<0 0 0>}
      localJointOffset=<10 0 0>
     }]
   }
   n23 = Anim Dynamics - Bone: hair7_1(ComponentPose=@n22, Alpha=1, ExternalForce="0,0,0") {
    angularDampingOverride=1
    angularSpringConstant=1500
    bAngularSpring=true
    bGravityOverrideInSimSpace=true
    bOverrideAngularDamping=true
    gravityScale=0.35
    simulationSpace=RootRelative
    alphaBoolBlend {blendOption=Linear}
    alphaScaleBiasClamp {scale=0.1}
    boundBone {boneName=hair7_1}
    componentLinearAccScale {y=0.46, z=0.46}
    componentLinearVelScale {y=0.46, z=0.46}
    physicsBodyDefinitions [{
      sphereCollisionRadius=0
      boundBone {boneName=hair7_1}
      boxExtents=<10 20 20>
      constraintSetup {angularLimitsMax=<0 90 90>, angularLimitsMin=<0 -90 -90>, angularTarget=<0 0 0>}
      localJointOffset=<10 0 0>
     }]
   }
   n24 = Anim Dynamics - Bone: hair7A_2(ComponentPose=@n23, Alpha=1, ExternalForce="0,0,0") {
    angularDampingOverride=1
    angularSpringConstant=1500
    bAngularSpring=true
    bGravityOverrideInSimSpace=true
    bOverrideAngularDamping=true
    gravityScale=0.35
    simulationSpace=RootRelative
    alphaBoolBlend {blendOption=Linear}
    alphaScaleBiasClamp {scale=0.2}
    boundBone {boneName=hair7A_2}
    componentLinearAccScale {y=0.46, z=0.46}
    componentLinearVelScale {y=0.46, z=0.46}
    physicsBodyDefinitions [{
      sphereCollisionRadius=0
      boundBone {boneName=hair7A_2}
      boxExtents=<10 20 20>
      constraintSetup {angularLimitsMax=<0 90 90>, angularLimitsMin=<0 -90 -90>, angularTarget=<0 0 0>}
      localJointOffset=<10 0 0>
     }]
   }
   n25 = Anim Dynamics - Bone: hair7B_2(ComponentPose=@n24, Alpha=1, ExternalForce="0,0,0") {
    angularDampingOverride=1
    angularSpringConstant=1500
    bAngularSpring=true
    bGravityOverrideInSimSpace=true
    bOverrideAngularDamping=true
    gravityScale=0.35
    simulationSpace=RootRelative
    alphaBoolBlend {blendOption=Linear}
    alphaScaleBiasClamp {scale=0.2}
    boundBone {boneName=hair7B_2}
    componentLinearAccScale {y=0.46, z=0.46}
    componentLinearVelScale {y=0.46, z=0.46}
    physicsBodyDefinitions [{
      sphereCollisionRadius=0
      boundBone {boneName=hair7B_2}
      boxExtents=<10 20 20>
      constraintSetup {angularLimitsMax=<0 90 90>, angularLimitsMin=<0 -90 -90>, angularTarget=<0 0 0>}
      localJointOffset=<10 0 0>
     }]
   }
   n26 = Anim Dynamics - Bone: hair7C_2(ComponentPose=@n25, Alpha=1, ExternalForce="0,0,0") {
    angularDampingOverride=1
    angularSpringConstant=1500
    bAngularSpring=true
    bGravityOverrideInSimSpace=true
    bOverrideAngularDamping=true
    gravityScale=0.35
    simulationSpace=RootRelative
    alphaBoolBlend {blendOption=Linear}
    alphaScaleBiasClamp {scale=0.2}
    boundBone {boneName=hair7C_2}
    componentLinearAccScale {y=0.46, z=0.46}
    componentLinearVelScale {y=0.46, z=0.46}
    physicsBodyDefinitions [{
      sphereCollisionRadius=0
      boundBone {boneName=hair7C_2}
      boxExtents=<10 20 20>
      constraintSetup {angularLimitsMax=<0 90 90>, angularLimitsMin=<0 -90 -90>, angularTarget=<0 0 0>}
      localJointOffset=<10 0 0>
     }]
   }
   Output Pose(Result=Component To Local(ComponentPose=@n26)) {layerGroup=None}
  ]}
 EventGraph {links [], nodes []}
]
variables [HeadRotation_b {type=number}]
